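# EQS query for the KeyDrop rift encounter that picks a location for a buried chest spawner.
# A donut generator lays spiral-pattern points around the buried chest context, with radii, ring count and points per ring bound to game difficulty curves.
# Points project using an any-traversable nav filter, searching mostly downward, then shift slightly below the surface.
# A single Fortnite random test scores the candidates, so placement ignores any other criteria.
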EnvQuery /RiftEncounters/Encounters/KeyDrop/EQS/EQS_Athena_BuriedChest_Spawner_KeyDrop
options [{
  generator {
   name=EnvQueryGenerator_Donut_1
   class=EnvQueryGenerator_Donut
   properties {
    bAutoSortTests=false
    bUseSpiralPattern=true
    center=/Game/Athena/Items/ForagedItems/BuriedChest/EQSC_Athena_BuriedChest.EQSC_Athena_BuriedChest_C
    arcAngle {dataBinding=null, dataField=FloatValue, defaultValue=360}
    arcDirection {dirMode=Rotation}
    innerRadius {
     dataBinding=&self:EnvQueryGenerator_Donut_1.FortAIDataProvider_FloatCurveOverGameDifficulty_0
     dataField=FloatValue
     defaultValue=256
    }
    numberOfRings {
     dataBinding=&self:EnvQueryGenerator_Donut_1.FortAIDataProvider_FloatCurveOverGameDifficulty_2
     dataField=None
     defaultValue=3
    }
    outerRadius {
     dataBinding=&self:EnvQueryGenerator_Donut_1.FortAIDataProvider_FloatCurveOverGameDifficulty_1
     dataField=FloatValue
     defaultValue=2560
    }
    pointsPerRing {
     dataBinding=&self:EnvQueryGenerator_Donut_1.FortAIDataProvider_FloatCurveOverGameDifficulty_3
     dataField=None
     defaultValue=3
    }
    projectionData {
     extentX=10
     navigationFilter=/Game/AI/NavFilters/NavFilter_AnyTraversable.NavFilter_AnyTraversable_C
     postProjectionVerticalOffset=-3
     projectDown=5000
     projectUp=100
    }
   }
  }
  tests [FortQueryTest_Random_1 {class=FortQueryTest_Random, properties {bUseRandomSeedForOthers=true, floatValueMax=1, scoringFactor=1}}]
 }]
